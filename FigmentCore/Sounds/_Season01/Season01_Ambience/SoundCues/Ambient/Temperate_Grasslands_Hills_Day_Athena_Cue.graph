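# Looping daytime outdoor ambience cue for temperate grassland hills, using Athena outdoor sound class and attenuation.
# A quality switch plays a three-layer mix on its first branch: two wind loops and a grassland day loop, each under endlessly looping volume and pitch envelopes.
# The other three quality branches share one randomly modulated wind loop, with nested envelopes of different lengths layering slow pitch and volume drift.
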SoundCue /FigmentCore/Sounds/_Season01/Season01_Ambience/SoundCues/Ambient/Temperate_Grasslands_Hills_Day_Athena_Cue
root=QualityLevel_1
nodes [
 {
  id=Enveloper_0
  children [WavePlayer_1]
  properties {
   bLoop=true
   bLoopIndefinitely=true
   loopEnd=18
   pitchCurve {editorCurveData {keys [{time=0}, {time=2.5155}, <at 5.238 1.0666>, <at 9.3217 1.3>, <at 12.5145 1.1047>, <at 18 1.0001>]}}
   volumeCurve {editorCurveData {keys [<at 0 0>,<at 5 1>,<at 6 1>,<at 8 0>]}}
  }
 }
 {
  id=Enveloper_1
  children [WavePlayer_0]
  properties {
   bLoop=true
   bLoopIndefinitely=true
   loopEnd=30
   pitchCurve {editorCurveData {keys [{time=0}]}}
   volumeCurve {editorCurveData {keys [<at 0 0>,<at 3 1>,<at 8.0559 0.9202>,<at 11.7509 0.4544>,<at 17 0>]}}
  }
 }
 {
  id=Enveloper_2
  children [WavePlayer_2]
  properties {
   bLoop=true
   bLoopIndefinitely=true
   loopEnd=30
   pitchCurve {editorCurveData {keys [{time=0}, <at 5 1.2>, {time=10}, <at 13.2262 1.0014>, <at 15.5363 0.9759>, <at 19.7025 0.8>, <at 26.2612 1.0014>, {time=30}]}}
   volumeCurve {editorCurveData {keys [<at 0 0.75>,<at 10 1.25>,<at 20 0.75>]}}
  }
 }
 {
  id=Enveloper_3
  children [Enveloper_4]
  properties {
   bLoop=true
   bLoopIndefinitely=true
   loopEnd=30
   pitchCurve {editorCurveData {keys [{interpMode=RCIM_Cubic, time=0}, <at 15.9468 0.9181 RCIM_Cubic>, {interpMode=RCIM_Cubic, time=30}]}}
   volumeCurve {editorCurveData {keys [<at 0 1 RCIM_Cubic>,<at 16.1421 0.921 RCIM_Cubic>,<at 30 1 RCIM_Cubic>]}}
  }
 }
 {
  id=Enveloper_4
  children [WavePlayer_3]
  properties {
   bLoop=true
   bLoopIndefinitely=true
   loopEnd=10
   pitchCurve {editorCurveData {keys [
      {time=0}
      {arriveTangent=0.0399, interpMode=RCIM_Cubic, leaveTangent=0.0399, time=1.9309}
      <at 3.9162 1.1563 RCIM_Cubic>
      {arriveTangent=-0.0257, interpMode=RCIM_Cubic, leaveTangent=-0.0257, time=6.0641}
      {time=10}
     ]}}
   volumeCurve {editorCurveData {keys [<at 0 1>]}}
  }
 }
 {id=Mixer_0, children [Enveloper_2,Enveloper_0,Enveloper_1], properties {inputVolume [0.7,0.4,0.4]}}
 {id=Modulator_0, children [Mixer_0], properties {pitchMax=1, pitchMin=1, volumeMax=1.6, volumeMin=1.6}}
 {id=Modulator_1, children [Enveloper_3], properties {pitchMax=1.1, pitchMin=0.9, volumeMax=1.1, volumeMin=0.9}}
 {id=QualityLevel_1, children [Modulator_0,Modulator_1,Modulator_1,Modulator_1]}
 {
  id=WavePlayer_0
  properties {bLooping=true, soundWaveAssetPtr=/FigmentCore/Sounds/_Season01/Fort_Ambient_Looping/Grassland_Exterior/Grassland_Day11c}
 }
 {
  id=WavePlayer_1
  properties {bLooping=true, soundWaveAssetPtr=/FigmentCore/Sounds/_Season01/Fort_Ambient_Looping/Wind_Exterior/Grasslands_Wind05}
 }
 {
  id=WavePlayer_2
  properties {bLooping=true, soundWaveAssetPtr=/FigmentCore/Sounds/_Season01/Fort_Ambient_Looping/Wind_Exterior/Grasslands_Wind04}
 }
 {
  id=WavePlayer_3
  properties {bLooping=true, soundWaveAssetPtr=&Grasslands_Wind04}
 }
]
properties {
 attenuationSettings=&Fort_Ambient_Athena_Att
 maxDistance=2097152
 soundClassObject=&Fort_Ambient_Outdoors_Athena
 volumeMultiplier=0.3
 preEffectBusSends [{}]
}
