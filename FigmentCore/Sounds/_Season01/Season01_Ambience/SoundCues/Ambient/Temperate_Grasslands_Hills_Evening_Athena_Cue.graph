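# Evening ambient sound cue for temperate grassland hills, using the outdoor Athena ambience sound class and attenuation.
# A quality-level node sends the first level to a mix of a pitched-down looping grassland bed and a spooky night sound, delayed 100 seconds and enveloped to swell and fade.
# The other three levels share just the grassland loop, with slight random volume and lowered pitch under an endlessly looping 30-second envelope.
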SoundCue /FigmentCore/Sounds/_Season01/Season01_Ambience/SoundCues/Ambient/Temperate_Grasslands_Hills_Evening_Athena_Cue
root=QualityLevel_1
nodes [
 {id=Delay_0, children [Enveloper_0], properties {delayMax=100, delayMin=100}}
 {
  id=Enveloper_0
  children [WavePlayer_1]
  properties {
   pitchCurve {editorCurveData {keys [{time=0}]}}
   volumeCurve {editorCurveData {keys [<at 0 0>,<at 5.8072 0.5594>,<at 12 1>,<at 15.1627 0.5726>,<at 18 0>]}}
  }
 }
 {
  id=Enveloper_1
  children [WavePlayer_2]
  properties {
   bLoop=true
   bLoopIndefinitely=true
   loopEnd=30
   pitchCurve {editorCurveData {keys [{interpMode=RCIM_Cubic, time=0}, <at 15.9468 0.9181 RCIM_Cubic>, {interpMode=RCIM_Cubic, time=30}]}}
   volumeCurve {editorCurveData {keys [<at 0 1 RCIM_Cubic>,<at 16.1421 0.921 RCIM_Cubic>,<at 30 1 RCIM_Cubic>]}}
  }
 }
 {id=Mixer_0, children [Modulator_0,Delay_0], properties {inputVolume [1,1]}}
 {id=Modulator_0, children [WavePlayer_0], properties {pitchMax=0.92, pitchMin=0.85}}
 {id=Modulator_1, children [Enveloper_1], properties {pitchMax=0.9, pitchMin=0.85, volumeMax=1.1, volumeMin=0.9}}
 {id=QualityLevel_1, children [Mixer_0,Modulator_1,Modulator_1,Modulator_1]}
 {
  id=WavePlayer_0
  properties {bLooping=true, soundWaveAssetPtr=/FigmentCore/Sounds/_Season01/Fort_Ambient_Looping/Grassland_Exterior/Grassland_Evening06}
 }
 {id=WavePlayer_1, properties {soundWaveAssetPtr=/FigmentCore/Sounds/_Season01/Fort_Ambient_Looping/Dread_Exterior/Spooky_Night07}}
 {
  id=WavePlayer_2
  properties {bLooping=true, soundWaveAssetPtr=&Grassland_Evening06}
 }
]
properties {
 attenuationSettings=&Fort_Ambient_Athena_Att
 maxDistance=2097152
 soundClassObject=&Fort_Ambient_Outdoors_Athena
 volumeMultiplier=0.3
 preEffectBusSends [{}]
}
