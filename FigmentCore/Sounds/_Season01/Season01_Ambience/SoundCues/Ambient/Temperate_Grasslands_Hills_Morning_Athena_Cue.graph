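# Looping outdoor ambience cue for morning in temperate grassland hills, using Athena outdoor sound class and attenuation.
# A quality-level switch gives the first slot a mix of birds, grassland dawn and wind, with envelopes fading the dawn bed and cyclically swelling and pitch-bending the wind.
# The other three slots share one lighter branch: the same bird loop, randomly pitched down, under a gentle looping envelope.
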SoundCue /FigmentCore/Sounds/_Season01/Season01_Ambience/SoundCues/Ambient/Temperate_Grasslands_Hills_Morning_Athena_Cue
root=QualityLevel_1
nodes [
 {
  id=Enveloper_0
  children [WavePlayer_2]
  properties {
   bLoop=true
   bLoopIndefinitely=true
   loopEnd=48
   pitchCurve {editorCurveData {keys [
      {time=0}
      <at 6.875 1.2>
      <at 13.2275 1.0001>
      <at 18.6725 0.85>
      <at 23.2705 1.0001>
      <at 29.9544 1.1028>
      <at 34.826 1.0001>
      <at 41.3462 0.8869>
      {time=47}
     ]}}
   volumeCurve {editorCurveData {keys [<at 0 0.1>,<at 4.3987 0.4372>,<at 8 1.0001>,<at 20.6512 0.1461>,<at 30 0.1>,<at 32.9224 1.1404>,<at 44 0.1>]}}
  }
 }
 {
  id=Enveloper_1
  children [WavePlayer_1]
  properties {
   pitchCurve {editorCurveData {keys [{time=0}]}}
   volumeCurve {editorCurveData {keys [<at 0 1>,<at 29.97 0.7266>,<at 59.175 0.3504>,<at 118.515 0.2464>]}}
  }
 }
 {
  id=Enveloper_2
  children [WavePlayer_3]
  properties {
   bLoop=true
   bLoopIndefinitely=true
   loopEnd=30
   pitchCurve {editorCurveData {keys [{interpMode=RCIM_Cubic, time=0}, <at 15.9468 0.9181 RCIM_Cubic>, {interpMode=RCIM_Cubic, time=30}]}}
   volumeCurve {editorCurveData {keys [<at 0 1 RCIM_Cubic>,<at 16.1421 0.921 RCIM_Cubic>,<at 30 1 RCIM_Cubic>]}}
  }
 }
 {id=Mixer_0, children [WavePlayer_0,Enveloper_1,Enveloper_0], properties {inputVolume [1,0.5,1]}}
 {id=Modulator_0, children [Mixer_0], properties {pitchMax=1, pitchMin=1, volumeMax=2, volumeMin=2}}
 {id=Modulator_1, children [Enveloper_2], properties {pitchMax=0.7, pitchMin=0.4, volumeMax=2, volumeMin=2}}
 {id=QualityLevel_1, children [Modulator_0,Modulator_1,Modulator_1,Modulator_1]}
 {
  id=WavePlayer_0
  properties {bLooping=true, soundWaveAssetPtr=/FigmentCore/Sounds/_Season01/Fort_Ambient_Looping/Nature_Exterior/mild_daytime_birds_06}
 }
 {
  id=WavePlayer_1
  properties {bLooping=true, soundWaveAssetPtr=/FigmentCore/Sounds/_Season01/Fort_Ambient_Looping/Grassland_Exterior/Grassland_Dawn02}
 }
 {
  id=WavePlayer_2
  properties {bLooping=true, soundWaveAssetPtr=/FigmentCore/Sounds/_Season01/Fort_Ambient_Looping/Wind_Exterior/Grasslands_Wind01}
 }
 {
  id=WavePlayer_3
  properties {bLooping=true, soundWaveAssetPtr=&mild_daytime_birds_06}
 }
]
properties {
 attenuationSettings=&Fort_Ambient_Athena_Att
 maxDistance=2097152
 soundClassObject=&Fort_Ambient_Outdoors_Athena
 volumeMultiplier=0.3
 preEffectBusSends [{}]
}
