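# Looping outdoor night ambience cue for temperate grassland hills in Athena, using the outdoor ambient sound class and Athena ambient attenuation.
# A quality-level node sends its first slot to a mix of pitched-down grassland dusk and evening loops with mountain wind.
# The other three slots share a single spooky night loop instead, shaped by an endlessly repeating 30-second pitch and volume envelope with slight random modulation.
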SoundCue /FigmentCore/Sounds/_Season01/Season01_Ambience/SoundCues/Ambient/Temperate_Grasslands_Hills_Night_Athena_Cue
root=QualityLevel_0
nodes [
 {
  id=Enveloper_0
  children [WavePlayer_2]
  properties {
   bLoop=true
   bLoopIndefinitely=true
   loopEnd=30
   pitchCurve {editorCurveData {keys [{interpMode=RCIM_Cubic, time=0}, <at 15.9468 0.9181 RCIM_Cubic>, {interpMode=RCIM_Cubic, time=30}]}}
   volumeCurve {editorCurveData {keys [<at 0 1 RCIM_Cubic>,<at 16.1421 0.921 RCIM_Cubic>,<at 30 1 RCIM_Cubic>]}}
  }
 }
 {id=Mixer_0, children [Modulator_0,WavePlayer_3,Modulator_1], properties {inputVolume [1,1,1]}}
 {id=Modulator_0, children [WavePlayer_0], properties {pitchMax=0.85, pitchMin=0.85}}
 {id=Modulator_1, children [WavePlayer_1], properties {pitchMax=0.7, pitchMin=0.7}}
 {id=Modulator_2, children [Mixer_0], properties {pitchMax=1, pitchMin=1, volumeMax=0.8, volumeMin=0.8}}
 {id=Modulator_3, children [Enveloper_0], properties {pitchMax=1.1, pitchMin=0.9, volumeMax=1.1, volumeMin=0.9}}
 {id=QualityLevel_0, children [Modulator_2,Modulator_3,Modulator_3,Modulator_3]}
 {
  id=WavePlayer_0
  properties {bLooping=true, soundWaveAssetPtr=/FigmentCore/Sounds/_Season01/Fort_Ambient_Looping/Grassland_Exterior/Grassland_Dusk01}
 }
 {
  id=WavePlayer_1
  properties {bLooping=true, soundWaveAssetPtr=/FigmentCore/Sounds/_Season01/Fort_Ambient_Looping/Grassland_Exterior/Grassland_Evening04}
 }
 {
  id=WavePlayer_2
  properties {bLooping=true, soundWaveAssetPtr=/FigmentCore/Sounds/_Season01/Fort_Ambient_Looping/Dread_Exterior/Spooky_Night02}
 }
 {
  id=WavePlayer_3
  properties {bLooping=true, soundWaveAssetPtr=/FigmentCore/Sounds/_Season01/Fort_Ambient_Looping/Mountain_Exterior/Mountain_Wind01}
 }
]
properties {
 attenuationSettings=&Fort_Ambient_Athena_Att
 maxDistance=2097152
 soundClassObject=&Fort_Ambient_Outdoors_Athena
 volumeMultiplier=0.3
 preEffectBusSends [{}]
}
